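# Additive, unlit, two-sided material for Niagara mesh particles, drawing animated tech lines for the Junior hologram table prototype.
# Vertex-color red steers world-position offset along normal and side axes, setting width and optionally sine-waved height.
# Opacity combines two opposite-panning procedural sine dash layers, subtracts a negative line, then applies erosion, fades and an optional gradient mask.
# Particle color drives HDR emissive; dynamic parameters scale width and main tiling.
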Material /Game/Athena/Prototype/Blueprints/Junior/M_Junior_HoloTable_TechLine
graph {
 properties {
  bUsedWithNiagaraMeshParticles=true
  blendMode=BLEND_Additive
  pixelDepthOffsetMode=PDOM_Legacy
  shadingModel=MSM_Unlit
  twoSided=true
  usedShadingModels=Unlit
 }
 nodes [
  Constant3Vector() {constant=<rgba 0 0.6112 1 0>}
  n1 = ParticleColor()
  n2 = ScalarParameter() {defaultValue=1, parameterName=Brightness}
  MaterialFunctionCall(Emissive=MaterialFunctionCall(LDR Input=Saturate(None=@n1.RGB), HDR Tint=Multiply(A=@n1.RGB, B=@n2)) {materialFunction=&DeriveHDRfromLDR}.RemappedRange) {materialFunction=&MF_FX_EmissiveAdjustments}
  n3 = VertexColor()
  n4 = ScalarParameter() {group=WPO, parameterName=Hight, sliderMax=1}
  n5 = TextureCoordinate()
  n6 = Add(
   A=Multiply(
    A=MaterialFunctionCall(Float2=@n5) {materialFunction=&BreakOutFloat2Components}.R
    B=ScalarParameter() {defaultValue=1, group=WPO, parameterName=HightSineWave_Tiling}
   )
   B=Multiply(A=Time(), B=ScalarParameter() {defaultValue=0.3, group=WPO, parameterName=HightSineWave_PanSpeed}) {constB=0.3}
  )
  n7 = Multiply(A=@n4, B=Sine(None=Multiply(A=@n6) {constB=0.5}))
  n8 = ScalarParameter() {group=WPO, parameterName=UseHightSineWave, sliderMax=1}
  n9 = Multiply(
   A=Multiply(
    A=VertexNormalWS()
    B=MaterialFunctionCall() {materialFunction=/Engine/Functions/Engine_MaterialFunctions02/WorldPositionOffset/ObjectScale}.Scale XYZ
   ) {constA=1}
   B=LinearInterpolate(Alpha=LinearInterpolate(A=@n4, B=@n7, Alpha=@n8)) {constB=20}
  ) {constB=20}
  n10 = Add(A=Multiply(A=@n3.R, B=@n9), B=Multiply(A=OneMinus(None=@n3.R), B=Multiply(A=@n9) {constB=-1}))
  n11 = LinearInterpolate(Alpha=Multiply(
    A=ScalarParameter() {defaultValue=0.5143, group=WPO, parameterName=Width}
    B=DynamicParameter() {paramNames [Width,Param2,Param3,Tile_Main]}.Width
   )) {constA=0.14}
  n12 = MaterialFunctionCall() {materialFunction=&ObjectScale}
  n13 = Multiply(A=CrossProduct(A=VertexTangentWS(), B=VertexNormalWS()), B=Multiply(A=Multiply(A=@n11, B=@n12.Scale XYZ) {constA=1}) {constB=20})
  n14 = VertexColor()
  n15 = Multiply(A=@n13, B=@n14.R)
  n16 = Multiply(A=@n13) {constB=-1}
  n17 = OneMinus(None=@n14.R)
  Multiply(A=Add(A=Add(A=@n10, B=@n15), B=Multiply(A=@n16, B=@n17)), B=ScalarParameter() {defaultValue=1, group=WPO, parameterName=WPO})
  n18 = MaterialFunctionCall() {materialFunction=&LinearGradient}
  n19 = Power(Base=@n18.UGradient, Exp=ScalarParameter() {defaultValue=1, group=CustomMask, parameterName=Mask_U_Bias}) {constExponent=1}
  n20 = Constant() {r=0.5}
  n21 = Power(Base=@n18.VGradient, Exp=ScalarParameter() {defaultValue=2, group=CustomMask, parameterName=Mask_V_Bias})
  NamedRerouteDeclaration(None=Multiply(
    A=SphereMask(A=@n19, B=@n20) {attenuationRadius=0.5, hardnessPercent=0}
    B=SphereMask(A=@n21, B=@n20) {attenuationRadius=0.5, hardnessPercent=0}
   )) {name=CustomMask, nodeColor=<rgba 1 0 0.9844 1>}
  Saturate(None=@n11)
  n22 = MaterialFunctionCall(Float2=TextureCoordinate()) {materialFunction=&BreakOutFloat2Components}
  n23 = ScalarParameter() {defaultValue=8, group=PanSpeed, parameterName=PanSpeed_Main}
  n24 = Add(A=@n22.R, B=Multiply(A=Time(), B=Divide(A=@n23) {constB=100}) {constB=0.08})
  n25 = Multiply(
   A=DynamicParameter() {paramNames [Width,Param2,Param3,Tile_Main]}.Tile_Main
   B=ScalarParameter() {defaultValue=34, group=Tiling, parameterName=Tile_Main}
  )
  n26 = Sine(None=Multiply(A=Multiply(A=@n24, B=@n25) {constB=10}) {constB=0.5})
  n27 = Sine(None=Multiply(A=@n22.G) {constB=0.5})
  n28 = Multiply(
   A=Multiply(A=Saturate(None=@n26), B=Saturate(None=@n27))
   B=LinearInterpolate(Alpha=ScalarParameter() {defaultValue=0.1, group=Hardness, parameterName=Hardness_Main, sliderMax=1}) {constB=30}
  )
  n29 = MaterialFunctionCall(Float2=TextureCoordinate()) {materialFunction=&BreakOutFloat2Components}
  n30 = Time()
  n31 = ScalarParameter() {defaultValue=2.5, group=PanSpeed, parameterName=PanSpeed_Detail}
  n32 = Add(A=@n29.R, B=Multiply(A=Multiply(A=@n30) {constB=-1}, B=Divide(A=@n31) {constB=100}))
  n33 = ScalarParameter() {defaultValue=500, group=Tiling, parameterName=Tile_Detail}
  n34 = Sine(None=Multiply(A=Multiply(A=@n32, B=@n33) {constB=10}) {constB=0.5})
  n35 = Sine(None=Multiply(A=@n29.G) {constB=0.5})
  n36 = Multiply(
   A=Multiply(A=Saturate(None=@n34), B=Saturate(None=@n35))
   B=LinearInterpolate(Alpha=ScalarParameter() {defaultValue=0.1, group=Hardness, parameterName=Hardness_Detail, sliderMax=1}) {constB=30}
  )
  n37 = MaterialFunctionCall(Float2=TextureCoordinate()) {materialFunction=&BreakOutFloat2Components}
  n38 = ScalarParameter() {defaultValue=1, parameterName=Offset_NegativeLine}
  n39 = Power(
   Base=SphereMask(A=Power(Base=@n37.G, Exp=@n38), B=Constant() {r=0.5}) {attenuationRadius=0.5, hardnessPercent=0}
   Exp=ScalarParameter() {defaultValue=1, parameterName=Contrast_NegativeLine}
  ) {constExponent=5}
  n40 = LinearInterpolate(Alpha=ScalarParameter() {group=Hardness, parameterName=Hardness_NegativeLine, sliderMax=1}) {constB=50}
  n41 = Subtract(A=Add(A=Saturate(None=@n28), B=Saturate(None=@n36)), B=Saturate(None=Multiply(A=@n39, B=@n40)))
  n42 = Saturate(None=ScalarParameter() {parameterName=Erode, sliderMax=1})
  n43 = Multiply(A=@n1.A, B=Clamp(None=Subtract(A=@n41, B=@n42), Min=ScalarParameter() {parameterName=MinClamp}))
  n44 = ScalarParameter() {parameterName=DepthFade}
  n45 = Multiply(
   A=ShadingPathSwitch(Default=DepthFade(Opacity=@n43, FadeDistance=@n44), Forward=@n43, Mobile=@n43)
   B=MaterialFunctionCall(Distance=ScalarParameter() {defaultValue=300, parameterName=NearCameraFade}) {materialFunction=&Near_Camera_Fade}.Result
  )
  StaticSwitchParameter(True=Multiply(A=NamedRerouteUsage(), B=@n45), False=@n45) {group=CustomMask, parameterName=CustomMask}
 ]
}
