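# Unlit translucent nebula material for Niagara mesh particles, in the Junior prototype folder.
# Color adds slowly panning, tiled Galaxy-skin stars to a panning nebula texture, tints by particle color, then runs through an emissive-adjustment function.
# Opacity uses a view-angle falloff that softens mesh silhouettes, with depth fade, particle alpha, near-camera fade and an optional mask texture.
# Dynamic parameters drive the falloff exponent and nebula desaturation.
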Material /Game/Athena/Prototype/Blueprints/Junior/M_Junior_Nebula
graph {
 properties {
  bUsedWithNiagaraMeshParticles=true
  blendMode=BLEND_Translucent
  pixelDepthOffsetMode=PDOM_Legacy
  shadingModel=MSM_Unlit
  usedShadingModels=Unlit
 }
 nodes [
  n1 = VertexNormalWS()
  n2 = CameraVectorWS()
  n3 = DynamicParameter()
  n4 = Power(Base=Abs(None=DotProduct(A=@n1, B=@n2)), Exp=@n3.Param1) {constExponent=5}
  n5 = ScalarParameter() {defaultValue=200, parameterName=DepthFade}
  n6 = ParticleColor()
  n7 = Multiply(
   A=Multiply(A=DepthFade(Opacity=@n4, FadeDistance=@n5), B=@n6.A)
   B=MaterialFunctionCall(Distance=ScalarParameter() {defaultValue=200, parameterName=NearCameraFade}) {materialFunction=&Near_Camera_Fade}.Result
  )
  n8 = TextureSampleParameter2D() {parameterName=MaskTex, samplerType=SAMPLERTYPE_LinearColor, texture=/Game/Global/Textures/Default/Masks_and_Gradients/RevealMask}
  LinearInterpolate(A=@n7, B=Multiply(A=@n7, B=@n8.R), Alpha=ScalarParameter() {parameterName=Use Mask Tex})
  n9 = TextureSample(UVs=Panner(Coordinate=TextureCoordinate() {uTiling=8, vTiling=8}) {speedX=-0.005, speedY=-0.005}) {samplerType=SAMPLERTYPE_Masks, texture=/Game/Characters/Player/Male/Medium/Bodies/M_Med_Soldier_04/Skins/Galaxy/Textures/Celestial_T_Stars}
  n10 = Divide(A=Add(A=@n9.R, B=@n9.R)) {constB=4}
  n11 = TextureCoordinate() {uTiling=1.2, vTiling=0.8}
  n12 = Desaturation(
   None=TextureSample(UVs=Panner(Coordinate=@n11) {speedX=-0.0015, speedY=-0.0015}) {texture=/Game/Characters/Player/Male/Medium/Bodies/M_Med_Soldier_04/Skins/Galaxy/Textures/T_FN_Nebula}.RGB
   Fraction=@n3.Param2
  )
  MaterialFunctionCall(Emissive=Multiply(A=Add(A=@n10, B=@n12), B=@n6.RGB)) {materialFunction=&MF_FX_EmissiveAdjustments}
 ]
}
